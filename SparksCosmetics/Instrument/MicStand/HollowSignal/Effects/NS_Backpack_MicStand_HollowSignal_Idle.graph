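# Idle Niagara effect for the Hollow Signal mic stand back bling, adding orange glows and lights to its skeletal mesh.
# One LightGlows emitter bursts two spawn groups of looping-lifetime particles, pinned to the mesh with separate stand and mic offsets.
# Spawn group selects color, size and light settings, driving two sprite renderers and a light renderer.
# Mesh scale sizes glow and light radius, while time of day or front-end preview sets glow alpha.
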NiagaraSystem /SparksCosmetics/Instrument/MicStand/HollowSignal/Effects/NS_Backpack_MicStand_HollowSignal_Idle
dependencies {
 usedDataInterfaces [&NiagaraDataInterfaceCurve]
 usedDynamicInputs [&MakeFloatFromVector, &Multiply_VectorByFloat]
 usedModules [
  &SystemState
  &self:SystemSpawnScript.SystemScriptSource.SystemScriptGraph.NiagaraNodeAssignment_2.SetVariables_#1
  &self:SystemSpawnScript.SystemScriptSource.SystemScriptGraph.NiagaraNodeAssignment_0.SetVariables_#2
 ]
}
emitters {LightGlows {
  emitterUpdateScript [
   &EmitterState()
   &SpawnBurst_Instantaneous(Use Spawn Probability=false, UseLoopCountLimit=true)
   &SpawnBurst_Instantaneous(Spawn Group=1, Use Spawn Probability=false, UseLoopCountLimit=true) {module=SpawnBurst_Instantaneous001}
  ]
  particleSpawnScript [
   &InitializeParticle(
    Color=&MergeHelper_ColorSwitch(0=<rgba 2 0.5426 0 1>, 1=<rgba 1 0.2713 0 1>, 2=?, 3=?, Index=&SpawnGroup(), Num=::NewEnumerator0)
    Lifetime=0.5
    Sprite Size Mode=::NewEnumerator3
    Uniform Sprite Size=&MergeHelper_FloatSwitch(0=35, 1=20, 2=?, 3=?, Index=&SpawnGroup(), Num=::NewEnumerator0)
    UsePositionOffset=false
    Write Material Random=true
   )
   &Light_Attributes(
    Diffuse Scale=1
    Falloff Exponent=1
    Light Diffuse Scale:NiagaraFloat=&MakeCustomFloatFromBool(
     Bool=$User.bIsFrontEndPreview
     True Float=50
     False Float=&MergeHelper_FloatSwitch(0=50, 1=25, 2=?, 3=?, Index=&SpawnGroup(), Num=::NewEnumerator0)
    )
    Light Exponent=2
    Light Radius:NiagaraFloat=&Multiply_Float(
     A=&MergeHelper_FloatSwitch(0=20, 1=15, 2=?, 3=?, Index=&SpawnGroup(), Num=::NewEnumerator0)
     B=$System.MeshScale
    )
    Radius=1
   )
   &CameraOffset(Camera Offset Amount=10)
   /SetVariables(
    Particles.GlowCameraOffset=-2
    Particles.GlowColor=&Multiply_LinearColorByFloat(
     Linear Color=$Particles.Color
     Float=&MakeCustomFloatFromBool(Bool=$User.bIsFrontEndPreview, True Float=1, False Float=1)
    )
   ) {isSetParameters=true}
   &DynamicMaterialParameters(
    Index 0 Param 1=4
    Index 0 Param 2=0.05
    Module.Index0Float0Write=true
    Module.Index0Float1Write=true
    Module.Index0Float2Write=false
    Module.Index0Float3Write=false
   )
  ]
  particleUpdateScript [
   &ParticleState(
    Kill Particles When Lifetime Has Elapsed=false
    Let Infinitely Lived Particles Die When Emitter Deactivates=true
    Loop Particles Lifetime=true
   )
   &ScaleSpriteSize(Uniform Curve Sprite Scale=DataInterface() {propertyValues {Curve {
       defaultValue=0
       keys [
        {interpMode=RCIM_Cubic, tangentMode=RCTM_User, time=0, value=0.8}
        {interpMode=RCIM_Cubic, tangentMode=RCTM_User, time=0.5, value=1}
        {tangentMode=RCTM_User, time=1, value=0.8}
       ]
      }}}) {enabled=false}
   &SkeletalMeshLocation(
    MaskSpawnGroup=false
    Mesh Sampling Type=::NewEnumerator5
    Orientation Sampling=::NewEnumerator1
    Sampled Position Offset=&MergeHelper_VectorSwitch(0=$System.StandOffset, 1=$System.MicOffset, 2=?, 3=?, Index=&SpawnGroup(), Num=::NewEnumerator0)
    Skeletal Mesh=$User.SkeletalMesh
    Socket Sampling Mode=::NewEnumerator5
   )
   /SetVariables(
    Particles.GlowColor=&MultiplyLinearColor(
     Linear Color=$Particles.Initial.GlowColor
     Scale Factor=&MakeVector4FromVectorAndFloat(
      "Vector (RGB)"=<1 1 1>
      "Float (Alpha)"=&MakeCustomFloatFromBool(
       Bool=$User.bIsFrontEndPreview
       True Float=0.75
       False Float=&FNTimeOfDayScalar(Night Scale=0.75, Morning Scale=1, Day Scale=0.5, Evening Scale=0.6)
      )
     )
    )
    Particles.GlowScale:Vector2f=&Multiply_Vector2D_ByFloat(Float=$System.MeshScale, Vector2D=$Particles.SpriteSize)
   ) {isSetParameters=true}
  ]
  rendererClasses [&NiagaraSpriteRendererProperties, &NiagaraLightRendererProperties]
  renderers [
   {rendererClass=&NiagaraSpriteRendererProperties}
   {rendererClass=&NiagaraLightRendererProperties, rendererIndex=1}
   {rendererClass=&NiagaraSpriteRendererProperties, rendererIndex=2}
  ]
 }}
summary {
 systemName=NS_Backpack_MicStand_HollowSignal_Idle
 emitters [{emitterName=LightGlows, rendererClasses [&NiagaraSpriteRendererProperties, &NiagaraLightRendererProperties]}]
 userVariables [
  User.bIsFrontEnd {type=NiagaraBool}
  User.bIsFrontEndPreview {type=NiagaraBool}
  User.SkeletalMesh {
   type=NiagaraDataInterfaceSkeletalMesh
   defaultValue {
    struct=NiagaraExt_VariableValue_DataInterface
    value {dataInterface=&self:NiagaraDataInterfaceSkeletalMesh_0}
   }
  }
 ]
}
